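# Translucent, unlit, two-sided Niagara sprite and mesh particle material for blob effects on the Dyed Duelist backpack.
# A tint color is multiplied by day or night intensity chosen through the time-of-day color function.
# A static switch selects the shape: a scissors texture mask with adjustable gradient power, or a procedural radial falloff from scaled UV distance, sharpened by a power.
# A separate BlobTransparency parameter is also present.
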Material /BRCosmetics/Accessories/FORT_Backpacks/Backpack_F_MED_DyedDuelist/FX/M_F_DyedDuelist_Blob_FX
graph {
 properties {
  bUsedWithNiagaraMeshParticles=true
  bUsedWithNiagaraSprites=true
  blendMode=BLEND_Translucent
  pixelDepthOffsetMode=PDOM_Legacy
  shadingModel=MSM_Unlit
  twoSided=true
  usedShadingModels=Unlit
 }
 nodes [
  n1 = ScalarParameter() {defaultValue=5, parameterName=DD_BlobIntensityDay}
  n2 = ScalarParameter() {defaultValue=5, parameterName=DD_BlobIntensityNight}
  Multiply(
   A=MaterialFunctionCall(Color_Morning=@n1, Color_Day=@n1, Color_Evening=@n2, Color_Night=@n2) {materialFunction=&Color_Time_of_Day}.Result
   B=VectorParameter() {parameterName=DD_BlobColor, defaultValue=<rgba 1 0 0.6305 1>}.RGB
  ) {constB=5}
  ScalarParameter() {parameterName=BlobTransparency}
  n3 = TextureSampleParameter2D() {parameterName=Mask, texture=&/T_DyedDuelist_Backpack_FX_Scissors}
  n4 = ScalarParameter() {defaultValue=1, parameterName=DD_BlobSize}
  n5 = Distance(A=Multiply(A=TextureCoordinate(), B=@n4), B=Divide(A=@n4))
  StaticSwitchParameter(
   True=Power(Base=@n3.RGB, Exp=ScalarParameter() {defaultValue=1, parameterName=DD_Gradient}) {constExponent=6}
   False=Power(Base=OneMinus(None=@n5)) {constExponent=6}
  ) {parameterName=IsTextureMask}
 ]
}
