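# Character cosmetic material for the Dyed Duelist backpack, using material attributes and an expression-driven shading model.
# It sandwiches a backpack-specific function, MF_DyedDuelist_Backpack, between the shared 2023 character parent PreFX and PostFX functions.
# PreFX feeds both later stages, and the backpack function's result goes into PostFX.
# Scalar parameters expose the vibrant emissive amount and an on-kill emissive amount to the backpack effect.
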Material /BRCosmetics/Accessories/FORT_Backpacks/Backpack_F_MED_DyedDuelist/Materials/M_DyedDuelist_Backpack
graph {
 properties {
  bUseMaterialAttributes=true
  bUsedWithMeshDeformer=true
  bUsedWithNanite=true
  bUsedWithSkeletalMesh=true
  pixelDepthOffsetMode=PDOM_Legacy
  shadingModel=MSM_FromMaterialExpression
  usedShadingModels=Default Lit | Subsurface | Clear Coat | Subsurface Profile | Cloth
 }
 nodes [
  n1 = MaterialFunctionCall() {materialFunction=&MF_CharacterParent_2023_PreFX}
  MaterialFunctionCall(
   Material=MaterialFunctionCall!(
    In=@n1.Result
    TAMask=@n1.Result
    OnKill=ScalarParameter() {parameterName=OnKillVibrantEmissiveAmount}
    VibrantEmissiveAmount=ScalarParameter() {parameterName=VibrantEmissiveAmount}
   ) {materialFunction=&/MF_DyedDuelist_Backpack}.Result
   M=@n1.M
   Pre Skinned Local Position=@n1.M
   Distance Blend=@n1.M
   Fresnel=@n1.M
  ) {materialFunction=&MF_CharacterParent_2023_PostFX}
 ]
}
