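# Physics asset for the DyedDuelist female medium back bling pack, defining simulated bodies for its dangling parts.
# Capsule bodies cover the pack, a large middle joint and two small left and two small right dynamic joints.
# Each dynamic joint is constrained directly to the pack rather than chained, with tight swing and twist limits on the side joints.
# The middle joint locks twist and allows a wide swing on one axis only.
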PhysicsAsset /BRCosmetics/Accessories/FORT_Backpacks/Backpack_F_MED_DyedDuelist/Meshes/F_MED_DyedDuelist_Pack_Physics
bodies {
 dyn_l_1_joint [{
   shapeType=Capsule
   length=3.8539
   radius=0.9241
   shapeName=dyn_l_1_joint_capsule
   center=<-0.8496 0.0019 0.0006>
   rotation=<rot 0 90.0374 -89.9891>
  }]
 dyn_l_2_joint [{
   shapeType=Capsule
   length=3.9482
   radius=0.9447
   shapeName=dyn_l_2_joint_capsule
   center=<-0.8576 0.0022 0.0004>
   rotation=<rot 0 90.0425 -89.9914>
  }]
 dyn_mid_joint [{
   shapeType=Capsule
   length=19.5905
   radius=6.7307
   shapeName=dyn_mid_joint_capsule
   center=<0.1172 -0.2429 0.4938>
   rotation=<rot 0.2994 -90 -0.3473>
  }]
 dyn_r_1_joint [{
   shapeType=Capsule
   length=3.8539
   radius=0.9241
   shapeName=dyn_r_1_joint_capsule
   center=<0.8496 -0.0019 -0.0006>
   rotation=<rot 0 90.0374 -89.9891>
  }]
 dyn_r_2_joint [{
   shapeType=Capsule
   length=3.9482
   radius=0.9447
   shapeName=dyn_r_2_joint_capsule
   center=<0.8576 -0.0022 -0.0004>
   rotation=<rot 0 90.0425 -89.9914>
  }]
 pack [{shapeType=Capsule, length=8, radius=7.4967, center=<-0.0528 -2.7153 -3.8274>, rotation=<rot -0.0004 -89.9937 -179.792>}]
}
constraints [
 {bone1Name=dyn_l_1_joint, bone2Name=pack, swing1LimitDegrees=17.76, swing2LimitDegrees=26.52, twistLimitDegrees=4}
 {bone1Name=dyn_l_2_joint, bone2Name=pack, swing1LimitDegrees=12, swing2LimitDegrees=15, twistLimitDegrees=2}
 {bone1Name=dyn_mid_joint, bone2Name=pack, swing1LimitDegrees=1.7, swing2LimitDegrees=45, twistLimitDegrees=45, twistMotion=Locked}
 {bone1Name=dyn_r_1_joint, bone2Name=pack, swing1LimitDegrees=12, swing2LimitDegrees=15, twistLimitDegrees=2}
 {bone1Name=dyn_r_2_joint, bone2Name=pack, swing1LimitDegrees=12, swing2LimitDegrees=15, twistLimitDegrees=2}
]
