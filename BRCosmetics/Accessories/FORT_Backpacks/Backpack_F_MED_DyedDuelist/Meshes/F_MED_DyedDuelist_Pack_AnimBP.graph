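# Animation Blueprint for the DyedDuelist back bling, adjusting the pack bone and adding secondary physics motion.
# The EventGraph reads owner mesh sockets, deriving skydiving rotation from spine_05 and melee yaw from an eased, amplified left clavicle rotation.
# AnimGraph chains additive pack-bone offsets, toggled by melee and skydiving flags, into a RigidBody node with custom gravity and clamped simulation-space motion.
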AnimBlueprint /BRCosmetics/Accessories/FORT_Backpacks/Backpack_F_MED_DyedDuelist/Meshes/F_MED_DyedDuelist_Pack_AnimBP
target_skeleton=&/F_MED_DyedDuelist_Pack_Skeleton
graphs [
 AnimGraph {nodes [
   n1 = "Transform (Modify) Bone - Bone: pack"!(
    ComponentPose=Local To Component(LocalPose=Input Pose())
    bAlphaBoolEnabled:bool=Get bIsPlayingMeleeAnim()
    Translation="0,0,0"
    Rotation_Roll:float=0
    Rotation_Pitch:float=0
    Rotation_Yaw:float=Get!() -> (Return Value Rotation_Roll:float, Return Value Rotation_Pitch:float, Return Value Rotation_Yaw:float).Return Value Rotation_Yaw
    Scale
   ) -> (Pose) {
    alphaInputType=Bool
    rotationMode=BMM_Additive
    rotationSpace=BCS_BoneSpace
    translationMode=BMM_Additive
    translationSpace=BCS_BoneSpace
    alphaBoolBlend {blendInTime=0.01, blendOption=Linear, blendOutTime=0.01}
    alphaScaleBiasClamp {outRange {min=-1}}
    boneToModify {boneName=pack}
   }
   n2 = Get bIsSkyDiving()
   n3 = "Transform (Modify) Bone - Bone: pack"(
    ComponentPose=@n1
    bAlphaBoolEnabled=@n2
    Translation="0,-1,0"
    Rotation_Roll=0
    Rotation_Pitch=0
    Rotation_Yaw=Get get_rotation() -> (get_rotation:float)
   ) {
    alphaInputType=Bool
    rotationMode=BMM_Additive
    translationMode=BMM_Additive
    alphaBoolBlend {blendInTime=-1, blendOption=Linear}
    alphaScaleBiasClamp {outRange {min=-1}}
    boneToModify {boneName=pack}
    translation {y=-1}
   }
   n4 = "Transform (Modify) Bone - Bone: pack"!(ComponentPose=@n3, bAlphaBoolEnabled=@n2, Translation="0,-3,1", Rotation:Rotator="0,0,2", Scale) -> (Pose) {
    alphaInputType=Bool
    rotationMode=BMM_Additive
    rotationSpace=BCS_BoneSpace
    translationMode=BMM_Additive
    translationSpace=BCS_BoneSpace
    alphaBoolBlend {blendInTime=-1, blendOption=Linear, blendOutTime=1}
    boneToModify {boneName=pack}
    rotation {roll=2}
    translation {y=-3, z=1}
   }
   n5 = RigidBody(ComponentPose=@n4) {
    bOverrideWorldGravity=true
    overridePhysicsAsset=&/F_MED_DyedDuelist_Pack_Physics
    alphaBoolBlend {blendOption=Linear}
    componentAppliedLinearAccClamp=<2000 2000 5000>
    componentLinearAccScale=<0.15 0.15 0.25>
    componentLinearVelScale=<0.5 0.5 0.5>
    overrideWorldGravity {z=-650}
    simSpaceSettings {maxAngularAcceleration=10, maxAngularVelocity=5, maxLinearAcceleration=3000, maxLinearVelocity=25}
   }
   Output Pose(Result=Component To Local(ComponentPose=@n5)) {layerGroup=None}
  ]}
 EventGraph {nodes [
   Event BlueprintUpdateAnimation()
   > n1 = Sequence()
   > Branch(Condition=Get bIsSkyDiving())
   n2 = GetOwnerSkeletalMeshComponent(self=Self-Reference())
   n3 = GetSocketTransform!(self=@n2, InSocketName=spine_05, TransformSpace=RTS_Component) -> (
    ReturnValue_Location_X:double
    ReturnValue_Location_Y:double
    ReturnValue_Location_Z:double
    ReturnValue_Rotation_Roll:float
    ReturnValue_Rotation_Pitch:float
    ReturnValue_Rotation_Yaw:float
    ReturnValue_Scale:Vector
   )
   > Set get_rotation(get_rotation:float=float / float(
     A=Invert Rotator!(A_Roll:float=@n3, A_Pitch:float=0, A_Yaw:float=0) -> (ReturnValue_Roll:float, ReturnValue_Pitch:float, ReturnValue_Yaw:float)
     B=4
    )) -> (Output_Get:float)
   n4 = Ease!(
    Function:EEasingFunc Enum=EaseIn
    Alpha:double=0.8
    A:Rotator=GetSocketTransform!(self=@n2, InSocketName=clavicle_l, TransformSpace=RTS_ParentBoneSpace) -> (ReturnValue_Location:Vector, ReturnValue_Rotation:Rotator, ReturnValue_Scale).ReturnValue_Rotation
    B:Rotator
    ShortestPath:bool=true
    BlendExp:double=0.5
   ) -> (Result_Roll:float, Result_Pitch:float, Result_Yaw:float) {is=title K2Node_EaseFunction}
   Set!(
    execute=@n1.then_1
    Return Value Rotation_Roll:float=@n4
    Return Value Rotation_Pitch:float=@n4.Result_Pitch
    Return Value Rotation_Yaw:float=float * float(A=@n4.Result_Yaw, B=3) {is=title K2Node_PromotableOperator}
   ) -> (then, Output_Get:Rotator)
   TryGetPawnOwner()
  ]}
]
variables [
 get_rotation {type=number}
 Return Value Rotation {type=Rotator}
]
